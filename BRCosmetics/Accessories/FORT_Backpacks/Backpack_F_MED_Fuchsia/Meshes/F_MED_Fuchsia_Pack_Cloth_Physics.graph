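# Physics asset for the Fuchsia backpack cosmetic's cloth mesh, defining capsule bodies for its dynamic knot and tassel bones.
# A larger master knot capsule anchors the left and right knot bodies and two mirrored two-bone tassel chains.
# Angular limits keep twist nearly locked, with the tassels given a wider overall swing range than the knot sides.
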PhysicsAsset /BRCosmetics/Accessories/FORT_Backpacks/Backpack_F_MED_Fuchsia/Meshes/F_MED_Fuchsia_Pack_Cloth_Physics
bodies {
 dyn_knot_MSTR [{
   shapeType=Capsule
   length=8.25
   radius=0.75
   shapeName=dyn_knot_MSTR_capsule
   center=<-0.4989 -1.4588 6.0561>
   rotation=<rot 2.0425 -0.6107 13.4202>
  }]
 dyn_knot_l [{shapeType=Capsule, length=3, radius=1, shapeName=dyn_knot_l_capsule, center=<1 0 0>, rotation=<rot 90 0 0>}]
 dyn_knot_r [{shapeType=Capsule, length=3, radius=1, shapeName=dyn_knot_r_capsule, center=<1 0 0>, rotation=<rot 90 0 0>}]
 dyn_tassle_l_01 [{shapeType=Capsule, length=3, radius=1, shapeName=dyn_tassle_l_01_capsule, center=<1 0 0>, rotation=<rot 90 0 0>}]
 dyn_tassle_l_02 [{shapeType=Capsule, length=3, radius=1, shapeName=dyn_tassle_l_02_capsule, center=<3 0 0>, rotation=<rot 90 0 0>}]
 dyn_tassle_r_01 [{shapeType=Capsule, length=3, radius=1, shapeName=dyn_tassle_r_01_capsule, center=<1 0 0>, rotation=<rot 90 0 0>}]
 dyn_tassle_r_02 [{shapeType=Capsule, length=3, radius=1, shapeName=dyn_tassle_r_02_capsule, center=<3 0 0>, rotation=<rot 90 0 0>}]
}
constraints [
 {bone1Name=dyn_knot_r, bone2Name=dyn_knot_MSTR, swing1LimitDegrees=16, swing2LimitDegrees=4, twistLimitDegrees=1}
 {bone1Name=dyn_knot_l, bone2Name=dyn_knot_MSTR, swing1LimitDegrees=16, swing2LimitDegrees=4, twistLimitDegrees=1}
 {bone1Name=dyn_tassle_r_01, bone2Name=dyn_knot_MSTR, swing1LimitDegrees=24, swing2LimitDegrees=24, twistLimitDegrees=1}
 {bone1Name=dyn_tassle_r_02, bone2Name=dyn_tassle_r_01, swing1LimitDegrees=24, swing2LimitDegrees=24, twistLimitDegrees=1}
 {bone1Name=dyn_tassle_l_01, bone2Name=dyn_knot_MSTR, swing1LimitDegrees=24, swing2LimitDegrees=24, twistLimitDegrees=1}
 {bone1Name=dyn_tassle_l_02, bone2Name=dyn_tassle_l_01, swing1LimitDegrees=24, swing2LimitDegrees=24, twistLimitDegrees=1}
]
