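# Physics asset for the Fuchsia backpack cosmetic, providing collision bodies and joints for its flag, knot and tassels.
# Five flat box bodies form a flag chain from a small master capsule, with swing2 locked and limited swing1 and twist at each joint.
# A knot master capsule anchors left and right knot capsules and two two-segment tassel chains, all with narrow swing and near-zero twist limits.
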PhysicsAsset /BRCosmetics/Accessories/FORT_Backpacks/Backpack_F_MED_Fuchsia/Meshes/F_MED_Fuchsia_Pack_Physics
bodies {
 dyn_flag_01 [{shapeType=Box, extentX=8.9477, extentY=2, extentZ=18.3336, center=<3.7665 -0.0001 -0.382>}]
 dyn_flag_02 [{shapeType=Box, extentX=11.0144, extentY=2, extentZ=18.145, center=<4.9291 0.0001 -0.335>, rotation=<rot 0.0001 0 -5>}]
 dyn_flag_03 [{shapeType=Box, extentX=6.3497, extentY=2, extentZ=19.6231, center=<2.693 0 0>, rotation=<rot 0 0 8>}]
 dyn_flag_04 [{shapeType=Box, extentX=7.3952, extentY=2, extentZ=14.2413, center=<3.0853 0.0001 0>, rotation=<rot 0 0 8>}]
 dyn_flag_05 [{shapeType=Box, extentX=5.09, extentY=2, extentZ=14.4261, center=<1.8479 -0.0001 0>, rotation=<rot 0 0 8>}]
 dyn_flag_MSTR [{shapeType=Capsule, length=0.505, radius=0.505, shapeName=dyn_flag_MSTR_capsule, rotation=<rot 0 -90 0>}]
 dyn_knot_MSTR [{
   shapeType=Capsule
   length=8.25
   radius=0.75
   shapeName=dyn_knot_MSTR_capsule
   center=<-0.4989 -1.4588 6.0561>
   rotation=<rot 2.0425 -0.6107 13.4202>
  }]
 dyn_knot_l [{shapeType=Capsule, length=3, radius=1, shapeName=dyn_knot_l_capsule, center=<1 0 0>, rotation=<rot 90 0 0>}]
 dyn_knot_r [{shapeType=Capsule, length=3, radius=1, shapeName=dyn_knot_r_capsule, center=<1 0 0>, rotation=<rot 90 0 0>}]
 dyn_tassle_l_01 [{shapeType=Capsule, length=3, radius=1, shapeName=dyn_tassle_l_01_capsule, center=<1 0 0>, rotation=<rot 90 0 0>}]
 dyn_tassle_l_02 [{shapeType=Capsule, length=3, radius=1, shapeName=dyn_tassle_l_02_capsule, center=<3 0 0>, rotation=<rot 90 0 0>}]
 dyn_tassle_r_01 [{shapeType=Capsule, length=3, radius=1, shapeName=dyn_tassle_r_01_capsule, center=<1 0 0>, rotation=<rot 90 0 0>}]
 dyn_tassle_r_02 [{shapeType=Capsule, length=3, radius=1, shapeName=dyn_tassle_r_02_capsule, center=<3 0 0>, rotation=<rot 90 0 0>}]
}
constraints [
 {bone1Name=dyn_knot_r, bone2Name=dyn_knot_MSTR, swing1LimitDegrees=16, swing2LimitDegrees=4, twistLimitDegrees=1}
 {bone1Name=dyn_knot_l, bone2Name=dyn_knot_MSTR, swing1LimitDegrees=16, swing2LimitDegrees=4, twistLimitDegrees=1}
 {bone1Name=dyn_tassle_r_01, bone2Name=dyn_knot_MSTR, swing1LimitDegrees=24, swing2LimitDegrees=24, twistLimitDegrees=1}
 {bone1Name=dyn_tassle_r_02, bone2Name=dyn_tassle_r_01, swing1LimitDegrees=24, swing2LimitDegrees=24, twistLimitDegrees=1}
 {bone1Name=dyn_tassle_l_01, bone2Name=dyn_knot_MSTR, swing1LimitDegrees=24, swing2LimitDegrees=24, twistLimitDegrees=1}
 {bone1Name=dyn_tassle_l_02, bone2Name=dyn_tassle_l_01, swing1LimitDegrees=24, swing2LimitDegrees=24, twistLimitDegrees=1}
 {bone1Name=dyn_flag_02, bone2Name=dyn_flag_01, swing1LimitDegrees=12, swing2LimitDegrees=1, swing2Motion=Locked, twistLimitDegrees=7}
 {bone1Name=dyn_flag_03, bone2Name=dyn_flag_02, swing1LimitDegrees=8, swing2LimitDegrees=8, swing2Motion=Locked, twistLimitDegrees=3}
 {bone1Name=dyn_flag_04, bone2Name=dyn_flag_03, swing1LimitDegrees=8, swing2LimitDegrees=40, swing2Motion=Locked, twistLimitDegrees=3}
 {bone1Name=dyn_flag_05, bone2Name=dyn_flag_04, swing1LimitDegrees=8, swing2LimitDegrees=50, swing2Motion=Locked, twistLimitDegrees=3}
 {bone1Name=dyn_flag_01, bone2Name=dyn_flag_MSTR, swing1LimitDegrees=12, swing2LimitDegrees=1, swing2Motion=Locked, twistLimitDegrees=7}
]
