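# Anim Blueprint for the F_MED_Fuchsia backpack that adds secondary physics motion to the pack and its flag.
# An offset ref pose feeds two RigidBody paths selected by bClothEnabled: cloth physics asset plus BwdMovement-weighted flag trail, or standard pack physics.
# Each update runs a deprecated state-based cloth controller, or sets GravityOverride when cloth is off.
# Zipline events trigger cloth reset-and-teleport; DBNO and Look At nodes sit outside the output chain.
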AnimBlueprint /BRCosmetics/Accessories/FORT_Backpacks/Backpack_F_MED_Fuchsia/Meshes/F_MED_Fuchsia_Pack_AnimBP
target_skeleton=&/F_MED_Fuchsia_Pack_Skeleton
graphs [
 AnimGraph {nodes [
   n1 = RigidBody(ComponentPose=Local To Component(LocalPose=Use cached pose 'Start_Cache'())) {
    overridePhysicsAsset=&/F_MED_Fuchsia_Pack_Cloth_Physics
    alphaBoolBlend {blendOption=Linear}
    baseBoneRef {boneName=pack}
    componentAppliedLinearAccClamp=<3000 3000 3000>
    componentLinearAccScale=<0.2 0.2 0.4>
    componentLinearVelScale=<0.1 0.1 0.2>
    overrideWorldGravity {z=-980}
   }
   n2 = Trail controller - Bone: dyn_flag_05!(ComponentPose=@n1, Alpha=1, RelaxationSpeedScale:float=2, FakeVelocity:Vector="0,0,0") -> (Pose) {
    bActorSpaceFakeVel=true
    bShowTrailLocation=true
    bUsePlanarLimit=true
    chainLength=5
    debugLifeTime=0.1
    maxDeltaTime=0.0333
    relaxationSpeedScale=1.2
    alphaScaleBiasClamp {bInterpResult=true, interpSpeedDecreasing=25, interpSpeedIncreasing=25}
    baseJoint {boneName=pack}
    rotationLimits [{limitMax=<180 180 180>}, {limitMax=<180 180 180>}, {limitMax=<180 180 180>}, {limitMax=<180 180 180>}, {limitMax=<180 180 180>}]
    rotationOffsets [<0 0 0>, <0 0 0>, <0 0 0>, <0 0 0>, <0 0 0>]
    trailBone {boneName=dyn_flag_05}
    trailRelaxationSpeed {editorCurveData {keys [<at 0 15 RCIM_Linear>,<at 1 6 RCIM_Linear>]}}
   }
   Blend Poses by bool(BlendPose_0=Component To Local(ComponentPose=@n2), bActiveValue=Get bIsDBNO()) {blendType=Linear}
   n3 = RigidBody(ComponentPose=Local To Component()) {
    overridePhysicsAsset=&/F_MED_Fuchsia_Pack_Physics
    alphaBoolBlend {blendOption=Linear}
    baseBoneRef {boneName=pack}
    componentAppliedLinearAccClamp=<3000 3000 3000>
    componentLinearAccScale=<0.2 0.2 0.4>
    componentLinearVelScale=<0.1 0.1 0.2>
   }
   Component To Local(ComponentPose=@n3)
   Look At - Bone: pack!(ComponentPose, Alpha=1, LookAtLocation:Vector=Get PelvisLocation() -> (PelvisLocation:Vector)) -> (Pose) {
    bUseLookUpAxis=true
    alphaBoolBlend {blendOption=Linear}
    boneToModify {boneName=pack}
    lookAt_Axis {bInLocalSpace=false, axis {y=0, z=-1}}
    lookAtTarget=~
   }
   n4 = Get GravityOverride()
   n5 = RigidBody!(ComponentPose=Local To Component(LocalPose=Use cached pose 'SavedPose19689'()), Alpha, OverrideWorldGravity:Vector=@n4, ExternalForce) -> (Pose) {
    bOverrideWorldGravity=true
    overridePhysicsAsset=&F_MED_Fuchsia_Pack_Physics
    alphaBoolBlend {blendOption=Linear}
    baseBoneRef {boneName=pack}
    componentAppliedLinearAccClamp=<3000 3000 3000>
    componentLinearAccScale=<0.2 0.2 0.4>
    componentLinearVelScale=<0.1 0.1 0.2>
   }
   n6 = RigidBody(ComponentPose=Local To Component(LocalPose=Use cached pose 'SavedPose19689'()), OverrideWorldGravity=@n4) {
    bOverrideWorldGravity=true
    overridePhysicsAsset=&F_MED_Fuchsia_Pack_Cloth_Physics
    alphaBoolBlend {blendOption=Linear}
    baseBoneRef {boneName=pack}
    componentAppliedLinearAccClamp=<3000 3000 3000>
    componentLinearAccScale=<0.2 0.2 0.4>
    componentLinearVelScale=<0.1 0.1 0.2>
   }
   n7 = Trail controller - Bone: dyn_flag_05!(ComponentPose=@n6, Alpha=Get BwdMovement() -> (BwdMovement:double)) -> (Pose) {
    bUsePlanarLimit=true
    chainLength=5
    debugLifeTime=0.1
    maxDeltaTime=0.1
    relaxationSpeedScale=0.33
    alphaScaleBiasClamp {bInterpResult=true}
    baseJoint {boneName=pack}
    planarLimits [{drivingBone {boneName=dyn_flag_MSTR}, planeTransform {location=<0 -3 0>, rotation=<rot 0 0 -90>, scale=<1 1 1>}}]
    rotationLimits [
     {limitMax=<30 30 30>, limitMin=<-30 -30 -30>}
     {limitMax=<30 30 30>, limitMin=<-30 -30 -30>}
     {limitMax=<30 30 30>, limitMin=<-30 -30 -30>}
     {limitMax=<30 30 30>, limitMin=<-30 -30 -30>}
     {limitMax=<30 30 30>, limitMin=<-30 -30 -30>}
    ]
    rotationOffsets [<0 0 0>, <0 0 0>, <0 0 0>, <0 0 0>, <0 0 0>]
    trailBone {boneName=dyn_flag_05}
    trailRelaxationSpeed {editorCurveData {keys [<at -0.0134 19.3841 RCIM_Linear>,<at 0.5079 12.8986 RCIM_Linear>,<at 0.9955 9.6044 RCIM_Linear>]}}
   }
   Output Pose(Result=Blend Poses by bool(
     BlendPose_0=Component To Local(ComponentPose=@n7)
     BlendPose_1=Component To Local(ComponentPose=@n5)
     BlendTime_0=0.2
     BlendTime_1=0.5
     bActiveValue=Get bClothEnabled()
    ) {blendType=Linear}) {layerGroup=None}
   n8 = "Transform (Modify) Bone - Bone: root"(ComponentPose=Local To Component(LocalPose=Local Space Ref Pose()), Translation="0,2.3,4", Rotation="0,0,0") {
    rotationMode=BMM_Additive
    translationMode=BMM_Additive
    alphaBoolBlend {blendOption=Linear}
    boneToModify {boneName=root}
    translation {y=2.3, z=4}
   }
   Save cached pose 'SavedPose19689'(Pose=Component To Local(ComponentPose=@n8)) {cacheName=SavedPose19689}
  ]}
 EventGraph {nodes [
   n1 = Event BlueprintUpdateAnimation()
   > n2 = Sequence!(execute) -> (then_0, then_1, then_2:Exec, then_3:Exec, then_4:Exec, then_5:Exec, then_6:Exec)
   > n3 = Branch(Condition=Get bClothEnabled())
   > n4 = Cloth Controller Function DEPRECATED!(
    execute
    CharacterPartAnimInstance:Custom Character Part Anim Instance Object Reference=Self-Reference()
    ClothingAssetName:String=Fuchsia_Flag_1
    Cloth Paramaters:Cloth Params Struct Structure=Get Cloth Paramaters() -> (Cloth Paramaters:Cloth Params Struct Structure)
    Cloth Core Settings:Cloth Core Setting Struct Structure=Get Cloth Core Settings() -> (Cloth Core Settings:Cloth Core Setting Struct Structure)
    ClothUpAxis:Vector="0,0,1"
    OnGround_Moving_SpeedLimit:double=2
    OnGround_Moving_AngularVel_Max:double=0.05
    OnGround_Moving_FictitiousVel_Max:double=0.6
    LinVelScale_FwdBwd_Clamp:double=0
    LinVelScale_UpDwn_Clamp:double=0.7
    DeltaTime:double=@n1.DeltaTimeX
    GameTimeInSeconds:double=GetGameTimeInSeconds()
    WindGustsParams:Wind Gust Params Struct Structure
    WindGustLocalSettings:Wind Gust Local Setting Struct Structure
   ) -> (
    then
    CurrentStateOut:String
    Cloth Settings Out:Cloth Setting Struct Structure
    Cloth Core Settings Out:Cloth Core Setting Struct Structure
    WindGustSettings Out:Wind Gust Setting Struct Structure
    WindGustLocalSettings Out:Wind Gust Local Setting Struct Structure
    GustStrength - WindVelocityOut:Vector
    LinearVelVector-DELETE LATER:Vector
   )
   n5 = Gravity Override State Function DEPRECATED(
    execute=@n3.else
    CharacterPartAnimInstance=Self-Reference()
    GameTimeInSeconds=GetGameTimeInSeconds()
    GravityOverrideParams=Get Gravity Override Params()
   )
   > Set GravityOverride(GravityOverride=@n5)
   n6 = GetPartSkeletalMeshComponent()
   Set TeleportDistanceThreshold!(execute=@n2.then_4, TeleportDistanceThreshold:float=300, self:Skeletal Mesh Component Object Reference=@n6) -> (then, Output_Get:float)
   > Set TeleportRotationThreshold!(execute, TeleportRotationThreshold:float=15, self:Skeletal Mesh Component Object Reference=@n6) -> (then, Output_Get:float)
   n7 = GetPartSkeletalMeshComponent()
   Is Valid(exec=@n2.then_5, InputObject=@n7)
   n8 = GetAttachParent!(self:Scene Component Object Reference=@n7) -> (ReturnValue:Scene Component Object Reference)
   > Is Valid(InputObject=@n8)
   > Set PelvisLocation(PelvisLocation:Vector=BreakTransform(InTransform=GetSocketTransform(self=@n8, InSocketName=spine_01, TransformSpace=RTS_World))) -> (Output_Get:Vector)
   n9 = MakeLiteralString(Value=Base_Cape_01)
   n10 = GetOwningComponent()
   n11 = Cast To ChaosClothingInteractor!(
    execute=@n2.then_3
    Object=GetClothingInteractor!(
     self:Clothing Simulation Interactor Object Reference=GetClothingSimulationInteractor!(self:Skeletal Mesh Component Object Reference=@n10, ClothingAssetName:Name=None) -> (ReturnValue:Clothing Simulation Interactor Object Reference)
     ClothingAssetName:Name=String To Name(InString=@n9)
    ) -> (ReturnValue:Clothing Interactor Object Reference)
   ) -> (then, CastFailed, AsChaos Clothing Interactor:Chaos Clothing Interactor Object Reference) {is=title K2Node_DynamicCast}
   > Set ChaosClothingInteractor(ChaosClothingInteractor:Chaos Clothing Interactor Object Reference=@n11) -> (Output_Get:Chaos Clothing Interactor Object Reference)
   Event BlueprintInitializeAnimation()
   > Set bResetAfterTeleport!(execute, bResetAfterTeleport:bool=false, self:Skeletal Mesh Component Object Reference=@n6) -> (then, Output_Get:bool)
   n12 = Event BlueprintProcessAnimInputEvent()
   > n13 = Branch(Condition=IsNamed(self=@n12.AnimInputEvent, InEventName=aie.Player.Zipline.Entered))
   > Delay(Duration=0.1)
   n14 = Get ChaosClothingInteractor() -> (ChaosClothingInteractor:Chaos Clothing Interactor Object Reference)
   > ResetAndTeleport!(execute, self:Chaos Clothing Interactor Object Reference=@n14, bReset:bool=true, bTeleport:bool=true) -> (then)
   Branch(execute=@n13.else, Condition=IsNamed(self=@n12.AnimInputEvent, InEventName=aie.Player.Zipline.Exited))
   > Set resetClothLandingFromZilpine(resetClothLandingFromZilpine:bool=true) -> (Output_Get:bool)
   Event OnEndFalling()
   > Branch(Condition=Get resetClothLandingFromZilpine() -> (resetClothLandingFromZilpine:bool))
   > ResetAndTeleport(self=@n14, bReset=true, bTeleport=true)
   > Set resetClothLandingFromZilpine(resetClothLandingFromZilpine=false)
   Event OnBeginSkydiving()
   > Set SkydivingAlpha(SkydivingAlpha:Rotator="0,-65,0") -> (Output_Get:Rotator)
   Event OnEndSkydiving()
   > Set SkydivingAlpha(SkydivingAlpha="0,-10,0")
   n15 = Branch(Condition=Get bIsSkyDiving())
   > n16 = Branch(Condition="float > float"(A=Get PawnVelocityZ(), B=50) {is=title K2Node_CallFunction})
   > Set BwdMovement(BwdMovement:double=float * float(A=Get BwdMovement(), B=90)) -> (Output_Get:double)
   Set BwdMovement(execute=[@n15.else, @n16.else], BwdMovement=float * float(A=Get BwdMovement(), B=0))
   n17 = Branch(Condition=Get bIsInFrontEnd())
   > Set BwdMovement(BwdMovement=0.5)
   n18 = PlayerMovement(execute=@n17.else, AnimInstance=Self-Reference(), XDotProductCoefficient=1, YDotProductCoefficient=1, ZDotProductCoefficient=1)
   > n19 = Branch(Condition=AND Boolean(A=Get bIsTargeting(), B=Get bIsCrouching()))
   n20 = MapRangeClamped(
    Value=float + float(A="Absolute (Float)"(A=@n18), B=MapRangeClamped(Value=@n18.YMovement, InRangeA=-1, InRangeB=1, OutRangeA=0.3)) {is=title K2Node_CommutativeAssociativeBinaryOperator}
    InRangeA=0
    InRangeB=1.65
   )
   > Set BwdMovement(BwdMovement=float * float(A=@n20, B=0.25) {is=title K2Node_PromotableOperator})
   Set BwdMovement(execute=@n19.else, BwdMovement=@n20)
   DrawVectorFromJointFunction!(
    execute
    SkelMeshComponent:Scene Component Object Reference=GetOwnerSkeletalMeshComponent()
    JointName:Name=pelvis
    LineEnd:Vector=Break Cloth Setting Struct!(ClothSettingStruct:"Cloth Setting Struct Structure (by ref)"=@n4.Cloth Settings Out) -> (
     LinearVelocityScale_7_#1:Vector=~
     AngularVelocityScale_6_#2:float=~
     FictitiousAngularScale_10_#3:float=~
     GravityOverride_15_#4:Vector=~
     DragCoefficient_23_#5:Vector 2D Structure
     LiftCoefficient_24_#6:Vector 2D Structure
     JointName_27_#7:Name
    ) {is=title K2Node_BreakStruct}.GravityOverride_15_#4
    JointOffset:Vector="0, 0, 0"
    LineColor:Linear Color Structure=(R=1,G=0,B=0,A=1)
    Thickness:double=2
    Foreground:EDrawDebugSceneDepthPriorityGroup Enum=Foreground
   ) -> (then)
   TryGetPawnOwner()
  ]}
]
variables [
 BwdMovement {type=number}
 ChaosClothingInteractor {type=ChaosClothingInteractor, default=null}
 CharacterSkelMesh {type=SkeletalMeshComponent}
 Cloth Core Settings {
  type=ClothCoreSettingStruct
  default {
   collisionThickness=1.5
   dampingCoefficient=0.1
   frictionCoefficient=0.2
   localDampingCoefficient=0.1
   numberIterations=2
   numberSubsteps=1
   selfCollisionThickness=0
   animDriveDamping=<0 1>
   animDriveStiffness=<0 1>
   areaStiffness=<1 1>
   bendingStiffness=<1 1>
   edgeStiffness=<1 1>
   tetherStiffness=<1 1>
  }
 }
 Cloth Paramaters {
  type=ClothParamsStruct
  default {
   dBNO {
    angularVelocityScale=0.6
    fictitiousAngularScale=0.1
    jointName=pelvis
    dragCoefficient=<0.2 1>
    gravityOverride=<-980 400 0>
    liftCoefficient=<0 0.5>
    linearVelocityScale=<0.3 0.3 0.1>
   }
   dBNO_Carried {
    angularVelocityScale=0.6
    fictitiousAngularScale=0.1
    jointName=pelvis
    dragCoefficient=<0.2 1>
    gravityOverride=<-980 -200 0>
    liftCoefficient=<0 0.5>
    linearVelocityScale=<0.3 0.3 0.1>
   }
   dBNO_Impulse {
    angularVelocityScale=0
    fictitiousAngularScale=0
    jointName=pelvis
    dragCoefficient=<0 0.2>
    gravityOverride=<-3000 -2000 0>
    liftCoefficient=<0 0.1>
    linearVelocityScale=<0.3 0.1 0.3>
   }
   dBNO_RollBwd {
    angularVelocityScale=0
    fictitiousAngularScale=0
    jointName=pelvis
    dragCoefficient=<0 0.2>
    gravityOverride=<-2000 0 0>
    liftCoefficient=<0 0.1>
    linearVelocityScale=<0.3 0.1 0.3>
   }
   dBNO_RollFwd {
    angularVelocityScale=0
    fictitiousAngularScale=0
    jointName=pelvis
    dragCoefficient=<0 0.2>
    gravityOverride=<-980 0 0>
    liftCoefficient=<0 0.1>
    linearVelocityScale=<0.3 0.1 0.3>
   }
   dBNO_Tumbling {
    angularVelocityScale=0
    fictitiousAngularScale=0
    jointName=pelvis
    dragCoefficient=<0 0.2>
    gravityOverride=<-2500 -300 0>
    liftCoefficient=<0 0.1>
    linearVelocityScale=<0.3 0.1 0.3>
   }
   emoteOrMelee {
    angularVelocityScale=0.6
    fictitiousAngularScale=0.1
    jointName=root
    dragCoefficient=<2.5 2.5>
    gravityOverride=<0 0 -980>
    liftCoefficient=<0.5 0.5>
    linearVelocityScale=<0.6 0.6 0.5>
   }
   front_End {
    angularVelocityScale=0.05
    fictitiousAngularScale=0.6
    jointName=root
    dragCoefficient=<0.2 0.2>
    gravityOverride=<0 -200 -980>
    liftCoefficient=<0.1 0.1>
    linearVelocityScale=<0.1 0.1 0.1>
   }
   onGround_Crouch_Moving {
    angularVelocityScale=0.1
    fictitiousAngularScale=0.3
    jointName=root
    dragCoefficient=<2 2>
    gravityOverride=<0 0 -980>
    liftCoefficient=<0.5 0.5>
    linearVelocityScale=<0.6 0.6 0.5>
   }
   onGround_Falling {
    angularVelocityScale=0.6
    fictitiousAngularScale=0
    jointName=root
    dragCoefficient=<0 0.2>
    gravityOverride=<0 -500 -980>
    liftCoefficient=<0 0.1>
    linearVelocityScale=<2 1 0.3>
   }
   onGround_Moving {
    angularVelocityScale=0.6
    fictitiousAngularScale=0.1
    jointName=root
    dragCoefficient=<2 2>
    gravityOverride=<0 0 -980>
    liftCoefficient=<0.5 0.5>
    linearVelocityScale=<0.3 0.3 0.25>
   }
   onGround_Standing {
    angularVelocityScale=0.05
    fictitiousAngularScale=1.2
    jointName=root
    dragCoefficient=<2.5 2.5>
    gravityOverride=<0 0 -980>
    liftCoefficient=<0.5 0.5>
    linearVelocityScale=<0.3 0.3 0.25>
   }
   rail_Grinding {
    angularVelocityScale=0.2
    fictitiousAngularScale=0
    jointName=root
    dragCoefficient=<0.1 0.2>
    gravityOverride=<50 -500 -400>
    liftCoefficient=<0 0.5>
    linearVelocityScale=<0.1 0.1 0.1>
   }
   riding_Moving {
    angularVelocityScale=0.05
    fictitiousAngularScale=1
    jointName=root
    dragCoefficient=<0.5 0.5>
    gravityOverride=<0 0 -980>
    liftCoefficient=<0.5 0.5>
    linearVelocityScale=<0.4 0.6 0.3>
   }
   riding_Moving_Targeting {
    angularVelocityScale=0.05
    fictitiousAngularScale=0.7
    jointName=root
    dragCoefficient=<0.5 0.5>
    gravityOverride=<0 0 -980>
    liftCoefficient=<0.5 0.5>
    linearVelocityScale=<0.2 0.6 0.2>
   }
   riding_Standing {
    angularVelocityScale=0.05
    fictitiousAngularScale=1
    jointName=root
    dragCoefficient=<0.5 0.5>
    gravityOverride=<0 -450 -980>
    liftCoefficient=<0.5 0.5>
    linearVelocityScale=<0.6 0.6 0.3>
   }
   skydive_DivingDown {
    angularVelocityScale=0.1
    fictitiousAngularScale=0
    jointName=pelvis
    dragCoefficient=<0.8 0.8>
    gravityOverride=<0 0 50>
    liftCoefficient=<2 2>
    linearVelocityScale=<0.1 0.1 0.1>
   }
   skydive_FloatingDown {
    angularVelocityScale=0.1
    fictitiousAngularScale=0
    jointName=root
    dragCoefficient=<1 1>
    gravityOverride=<0 -25 100>
    liftCoefficient=<2 2>
    linearVelocityScale=<0.2 0.2 0.1>
   }
   skydive_GoingUp {
    angularVelocityScale=0.1
    fictitiousAngularScale=0
    jointName=root
    dragCoefficient=<0.01 0.01>
    gravityOverride=<0 -2000 400>
    liftCoefficient=<0.05 0.05>
    linearVelocityScale=<0.1 0.1 0.01>
   }
   skydive_Parachuting {
    angularVelocityScale=0.6
    fictitiousAngularScale=0.1
    jointName=root
    dragCoefficient=<0.8 0.8>
    gravityOverride=<0 0 -780>
    liftCoefficient=<0.8 0.8>
    linearVelocityScale=<0.2 0.2 0.3>
   }
   skydive_Surfing {
    angularVelocityScale=0.6
    fictitiousAngularScale=0.1
    jointName=root
    dragCoefficient=<1.5 1.5>
    gravityOverride=<0 -600 -500>
    liftCoefficient=<1 1>
    linearVelocityScale=<0.1 0.1 0.3>
   }
   swimming {
    angularVelocityScale=0.6
    fictitiousAngularScale=0
    jointName=root
    dragCoefficient=<2.5 2.5>
    gravityOverride=<0 0 0>
    liftCoefficient=<0.1 0.1>
    linearVelocityScale=<0.6 0.6 0.5>
   }
   tactical_Sprint {
    angularVelocityScale=0.6
    fictitiousAngularScale=0.1
    jointName=root
    dragCoefficient=<2 2>
    gravityOverride=<0 0 -980>
    liftCoefficient=<0.5 0.5>
    linearVelocityScale=<0.6 0.6 0.5>
   }
   ziplining {
    angularVelocityScale=0.6
    fictitiousAngularScale=0
    jointName=root
    dragCoefficient=<0 1.5>
    gravityOverride=<0 0 -980>
    liftCoefficient=<0 0.3>
    linearVelocityScale=<0.4 0.4 0.5>
   }
  }
 }
 compLinearAccScale {type=Vector}
 compLinearVelScale {type=Vector}
 finalGravity {type=Vector}
 Gravity Override Params {
  type=GravityOverrideParamsStruct
  default {
   emoteOrMelee {windFrequency=<0.6 0.6 0.5>}
   front_End {gravityOverride=<0 -200 -980>, windFrequency=<0.1 0.1 0.1>}
   onGround_Moving {windFrequency=<0.6 0.6 0.3>}
   onGround_Standing {windFrequency=<0.6 0.6 0.3>}
   skydive_DivingDown {gravityOverride=<0 0 100>, windFrequency=<0.1 0.1 0.1>}
   skydive_FloatingDown {gravityOverride=<0 -1000 50>, windAmplitude=<100 50 200>, windFrequency=<0.1 0.1 0.1>, windOffset=<0 50 200>}
   skydive_GoingUp {jointName=pelvis, gravityOverride=<-1000 0 0>, windAmplitude=<200 -50 400>, windFrequency=<0.01 0.1 0.1>}
   skydive_Parachuting {gravityOverride=<0 0 -780>, windFrequency=<0.1 0.1 0.1>}
   swimming {jointName=root, gravityOverride=<0 0 0>, windFrequency=<0.6 0.6 0.5>}
  }
 }
 GravityOverride {type=Vector, default=<0 0 -980>}
 MeleeAlpha {type=number}
 PelvisLocation {type=Vector}
 resetClothLandingFromZilpine {type=boolean}
 SkydivingAlpha {type=Rotator}
 windVector {type=Vector}
]
